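# Unlit, translucent, two-sided glow material for Niagara mesh particles in the Cherry female pickaxe effects.
# Emissive combines smoke, squared panning noise and Fresnel, tinted by color parameter and particle color, then passes through MF_FX_EmissiveAdjustments.
# Opacity fades along V of the second UV set, and a vertex-normal offset inflates the mesh.
# Only_Use_Torso blends all three toward a simpler torso variant, whose emissive uses a color-change texture and Fresnel.
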Material /BRCosmetics/Weapons/FORT_Melee/Pickaxe_Cherry_Female/FX/M_Pickaxe_Cherry_Mesh_Glow
graph {
 properties {
  bUsedWithNiagaraMeshParticles=true
  blendMode=BLEND_Translucent
  pixelDepthOffsetMode=PDOM_Legacy
  shadingModel=MSM_Unlit
  twoSided=true
  usedShadingModels=Unlit
 }
 nodes [
  Constant2Vector() {r=-0.2}
  n1 = ParticleColor()
  n2 = ComponentMask(None=TextureCoordinate() {coordinateIndex=1}) {g=true}
  n3 = Multiply(A=Power(Base=OneMinus(None=@n2), Exp=ScalarParameter() {defaultValue=8, parameterName=Torso_Bottom_mask}) {constExponent=8}) {constB=5}
  n4 = ParticleColor()
  n5 = TextureSampleParameter2D(UVs=TextureCoordinate() {coordinateIndex=1}) {
   parameterName=Texture_B
   samplerType=SAMPLERTYPE_Masks
   texture=/Game/Characters/Player/Male/Medium/Bodies/M_MED_Storm_Soldier/Textures/ElectricityNoise_M
  }
  n6 = ScalarParameter() {defaultValue=3, parameterName=Fresnel_outside}
  n7 = ScalarParameter() {defaultValue=0.02, parameterName=Fresnel_inside}
  n8 = Add(
   A=Multiply(A=@n5.R) {constB=0.1}
   B=Multiply(
    A=Fresnel(ExponentIn=@n6, BaseReflectFractionIn=@n7) {baseReflectFraction=0.02, exponent=3}
    B=ScalarParameter() {defaultValue=0.5, parameterName=FresnelAffectTexture}
   ) {constA=1}
  )
  n9 = ComponentMask(None=TextureCoordinate() {coordinateIndex=1}) {g=true}
  n10 = Multiply(A=Power(Base=OneMinus(None=@n9), Exp=ScalarParameter() {defaultValue=4, parameterName=MaskPower}) {constExponent=4}) {constB=5}
  LinearInterpolate(
   A=Multiply(A=Multiply(A=@n4.A, B=@n8), B=Clamp(None=@n10))
   B=Multiply(A=@n1.A, B=Clamp(None=@n3))
   Alpha=ScalarParameter() {parameterName=Only_Use_Torso}
  )
  LinearInterpolate(
   A=Multiply(A=VertexNormalWS(), B=ScalarParameter() {defaultValue=0.2, parameterName=VertexNormal_WPO}) {constB=0.2}
   B=Multiply(A=VertexNormalWS()) {constB=0.1}
   Alpha=ScalarParameter() {parameterName=Only_Use_Torso}
  )
  n11 = Multiply(A=TextureSample() {texture=&/T_Pickaxe_Cherry_Torso_Color_Change}.RGB, B=@n1.RGB)
  n12 = Fresnel() {baseReflectFraction=0.1}
  n13 = VectorParameter() {parameterName=Color, defaultValue=<rgba 30 19.9774 16.3498 1>}
  n14 = VectorParameter() {parameterName=UV_offset, defaultValue=<rgba -0.2 0 0 0>}
  n15 = Add(A=TextureCoordinate() {coordinateIndex=1, uTiling=0.5, vTiling=0.5}, B=AppendVector(A=@n14.R, B=@n14.R))
  n16 = VectorParameter() {parameterName=UV_Scale, defaultValue=<rgba 1 1 0 0>}
  n17 = AppendVector(A=@n16.R, B=@n16.R)
  n18 = TextureCoordinate() {coordinateIndex=1, vTiling=0.5}
  n19 = TextureSample(UVs=Panner(Coordinate=Multiply(A=@n17, B=@n18), Time=ScalarParameter() {defaultValue=1.5, parameterName=Panner_offset}) {speedX=0.1, speedY=0.3}) {samplerType=SAMPLERTYPE_LinearColor, texture=&T_NoisePack_LinearColor}
  n20 = Multiply(A=Multiply(A=@n19.RGB, B=@n19.RGB))
  n21 = Multiply(
   A=Multiply(
    A=TextureSampleParameter2D(UVs=@n15) {parameterName=Texture_A, texture=&T_Inky_Smoke_Tile_Chromatic}.R
    B=ScalarParameter() {defaultValue=1, parameterName=Texture_A_intersity}
   )
   B=@n20
  )
  n22 = Constant3Vector() {constant=<rgba 1 0.6248 0.1506 0>}
  n23 = Add(A=Desaturation(None=Multiply(A=@n21) {constB=10}, Fraction=Constant() {r=-0.5}), B=Multiply(A=Multiply(A=@n20, B=@n22), B=@n8))
  n24 = Multiply(A=@n4.RGB, B=Desaturation(None=Multiply(A=@n23) {constB=10}, Fraction=Constant() {r=0.3}))
  MaterialFunctionCall(Emissive=LinearInterpolate(A=Multiply(A=@n13.RGB, B=@n24), B=Multiply(A=@n11, B=@n12), Alpha=ScalarParameter() {parameterName=Only_Use_Torso})) {materialFunction=&MF_FX_EmissiveAdjustments}
 ]
}
